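# Swamp terrain material function that twists flowing UVs into swirls, one per cell of a staggered grid.
# Time, direction and speed scroll the UVs, then a squared sphere mask per cell drives rotation about that cell's center.
# A tiled grid pseudo-random value flips spin direction and phases a sine-animated per-block output; Low quality drops both variations.
# Outputs include rotated UVs, the mask and auxiliary values; a sine-scaled UV branch is unconnected.
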MaterialFunction /Game/Packages/DS_Fortnite_TRV_Terrain/Materials/Swamp_TwistFunction
graph {nodes [
  n1 = FunctionInput() {bUsePreviewValueAsDefault=true, inputName=Time offset, inputType=FunctionInput_Scalar, previewValue=<0 1 0 1>}
  n2 = FunctionInput(Preview=Time()) {bUsePreviewValueAsDefault=true, inputName=un frac Time, inputType=FunctionInput_Scalar, previewValue=<0 1 0 1>}
  n3 = Add(A=@n1, B=@n2)
  n4 = FunctionInput() {bUsePreviewValueAsDefault=true, inputName=Flow Direction, inputType=FunctionInput_Vector2, previewValue=<0 1 0 1>}
  n5 = FunctionInput() {bUsePreviewValueAsDefault=true, inputName=Flow Speed, inputType=FunctionInput_Scalar, previewValue=<1 1 0 1>}
  n6 = FunctionInput() {bUsePreviewValueAsDefault=true, inputName=UVS * size, inputType=FunctionInput_Vector2, previewValue=<128 128 0 1>}
  n7 = Add(A=Multiply(A=Multiply(A=@n3, B=@n4), B=@n5), B=@n6) {constA=0.5}
  n8 = ComponentMask(None=@n7) {r=true}
  n9 = Ceil(None=Add(A=@n8) {constB=0.5})
  n10 = ComponentMask(None=@n7) {g=true}
  n11 = AppendVector(A=@n8, B=Add(A=Multiply(A=@n9) {constB=0.5}, B=Multiply(A=@n10)))
  n12 = Add(A=@n11) {constB=0.5}
  n13 = SphereMask(A=Frac(None=@n12), B=Constant() {r=0.5}) {attenuationRadius=0.5, desc=23, hardnessPercent=0}
  n14 = Multiply(A=@n13, B=@n13)
  n15 = FunctionInput() {bUsePreviewValueAsDefault=true, inputName=Grid Noise Tile Rep, inputType=FunctionInput_Vector2, previewValue=<5 3 0 1>}
  n16 = Frac(None=Divide(A=Add(A=@n11) {constB=-0.5}, B=@n15) {constB=5}) {desc=mod}
  n17 = Divide(A=Floor(None=Multiply(A=@n16, B=@n15) {constB=10}), B=@n15) {desc=0-1 values that have the number of increments specified}
  n18 = Frac(None=Add(A=ComponentMask(None=@n17) {g=true}, B=ComponentMask(None=@n17) {r=true}))
  n19 = LinearInterpolate(Alpha=Ceil(None=Subtract(A=@n18) {constB=0.5})) {constA=-1}
  n20 = Multiply(A=@n14, B=@n19) {desc=4 ins to randomize}
  n21 = FunctionInput() {bUsePreviewValueAsDefault=true, inputName=Rotation amount modulator, inputType=FunctionInput_Scalar, previewValue=<0.5 0 0 1>}
  n22 = Ceil(None=@n12)
  n23 = Constant()
  n24 = RotateAboutAxis(
   NormalizedRotationAxis=Constant3Vector() {constant=<rgba 0 0 1 0>}
   RotationAngle=Multiply(A=QualitySwitch(Default=@n20, Low=@n14, High=@n20), B=@n21) {constA=1}
   PivotPoint=AppendVector(A=Subtract(A=@n22) {constB=0.5}, B=@n23)
   Position=AppendVector(A=@n12, B=@n23)
  )
  FunctionOutput(None=ComponentMask(None=@n24) {g=true, r=true}) {outputName=Rotated UVs, sortPriority=1}
  FunctionOutput(None=@n14) {outputName=RotationAngleValues, sortPriority=1}
  FunctionOutput(None=@n12) {outputName=NIU - Modified UVOutput, sortPriority=1}
  FunctionOutput(None=@n19) {outputName=NIU - Random values negative or positive sign, sortPriority=30}
  n25 = Constant() {r=1}
  n26 = FunctionInput() {bUsePreviewValueAsDefault=true, inputName=SineAnimSpeed, inputType=FunctionInput_Scalar, previewValue=<0.3 3 0 1>}
  n27 = Multiply(A=@n2, B=@n26)
  n28 = FunctionInput() {bUsePreviewValueAsDefault=true, inputName=SineAnimOffset, inputType=FunctionInput_Scalar, previewValue=<0 3 0 1>}
  n29 = Sine(None=Add(A=@n18, B=Add(A=@n27, B=@n28)) {desc=offset wave phase})
  n30 = Multiply(A=@n25, B=ConstantBiasScale(None=@n29)) {desc=+15 for random height}
  FunctionOutput(None=QualitySwitch(Default=@n30, Low=@n25, High=@n30)) {outputName=NIU - Random values per block, sortPriority=30}
  n31 = Multiply(A=@n24, B=ConstantBiasScale(None=@n29) {bias=2, scale=0.25}) {desc=+15 for random height}
  QualitySwitch(Default=@n31, Low=Multiply(A=@n24) {constB=0.75}, High=@n31)
  TextureCoordinate() {desc=Tester, uTiling=5, vTiling=5}
 ]}
