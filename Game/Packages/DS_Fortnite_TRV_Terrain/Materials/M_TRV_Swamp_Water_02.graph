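# Opaque lit swamp water material for Transylvania terrain, blending world-aligned dead plants, grassy ground and murky water with duckweed.
# A shoreline mask from UV1, ground alpha and noise separates land from water; vertex green and pixel-depth fades shape leaf coverage.
# Twist-distorted, opposing panned normals ripple the water and vary its color, while vertex red drives duckweed.
# Fresnel is removed on clear water, and attributes feed the water render-to-texture function.
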Material /Game/Packages/DS_Fortnite_TRV_Terrain/Materials/M_TRV_Swamp_Water_02
graph {
 properties {
  bDisableDepthTest=true
  bTangentSpaceNormal=false
  bUseMaterialAttributes=true
  bUsedWithNanite=true
  pixelDepthOffsetMode=PDOM_Legacy
  usedShadingModels=Default Lit
 }
 nodes [
  ComponentMask(None=TextureCoordinate()) {desc=Tester, r=true}
  ComponentMask(None=TextureCoordinate()) {desc=Tester, r=true}
  n1 = CollectionParameter() {
   collection=&FortniteMaterialParameters
   parameterName=FortniteActiveTimeOfDay
  }
  ComponentMask(None=@n1) {b=true, desc=B - Evening}
  Constant()
  Constant() {r=0.4}
  n2 = MaterialFunctionCall!(
   TextureObject=TextureObject() {texture=/Game/Packages/DS_Fortnite_TRV_Terrain/Textures/T_TRV_Deadplants_01_D}
   TextureSize=Abs(None=ScalarParameter() {defaultValue=768, group=Mud, parameterName=CoastTextureTiling})
   WorldPosition
   Export Float 4=StaticBool() {value=true}
   World Space Normal
   MipBias=ScalarParameter() {defaultValue=-2, parameterName=LeavesMipBias}
   ProjectionTransitionContrast
  ) {materialFunction=/Engine/Functions/Engine_MaterialFunctions01/Texturing/WorldAlignedTexture_MipBias}
  n3 = ComponentMask(None=@n2.Z Texture) {a=true}
  n4 = ScalarParameter() {defaultValue=5, parameterName=Deadplants Mult 1}
  n5 = Clamp(None=OneMinus(None=Multiply(A=@n3, B=@n4)))
  n6 = Multiply(A=OneMinus(None=@n3), B=ScalarParameter() {defaultValue=6, parameterName=Deadplants Mult 2})
  n7 = ScalarParameter() {defaultValue=1.5, parameterName=Deadplants Pow 1}
  n8 = OneMinus(None=Min(A=VertexColor().G) {constB=0.99})
  n9 = Add(A=Multiply(A=Power(Base=@n6, Exp=@n7), B=@n8), B=@n8)
  n10 = LinearInterpolate(A=Clamp(None=Add(A=@n5, B=@n9)), Alpha=NamedRerouteUsage())
  n11 = MaterialFunctionCall(
   TextureObject=TextureObject() {texture=/Game/Packages/DS_Fortnite_TRV_Terrain/Textures/T_Transylvania_HF_Grass_2_D}
   TextureSize=Abs(None=ScalarParameter() {defaultValue=768, group=Mud, parameterName=GroundTextureTiling})
   Export Float 4=StaticBool() {value=true}
  ) {materialFunction=&WorldAlignedTexture}
  n12 = OneMinus(None=ComponentMask(None=@n11.Z Texture) {a=true}) {desc=alpha from Ground diffuse}
  n13 = Constant() {r=0.5}
  n14 = Clamp(None=Multiply(A=Power(Base=@n12, Exp=@n13), B=Constant() {r=1}))
  n15 = ComponentMask(None=WorldPosition() {worldPositionShaderOffset=WPT_ExcludeAllShaderOffsets}.XYZ) {g=true, r=true}
  n16 = Multiply(
   A=ScalarParameter() {defaultValue=1, parameterName=Ground Mult 1}
   B=TextureSample(UVs=Divide(A=@n15) {constB=800, desc=Scale of shoreline distortion}) {desc=Texture for shoreline distortion, texture=/Game/Packages/DS_Fortnite_TRV_Terrain/Textures/T_TRV_SwampWater_02_D}.A
  )
  n17 = TextureCoordinate() {coordinateIndex=1, uTiling=4, vTiling=4}
  n18 = Constant() {r=3}
  n19 = Clamp(None=ComponentMask(None=Add(A=@n17, B=@n18)) {g=true})
  n20 = Multiply(A=Add(A=Add(A=@n14, B=@n16), B=@n19), B=@n19)
  n21 = ScalarParameter() {defaultValue=5, parameterName=Ground Pow 1}
  n22 = Clamp(None=Multiply(A=Power(Base=@n20, Exp=@n21), B=ScalarParameter() {defaultValue=15, parameterName=Ground Mult 2}))
  n23 = Multiply(A=@n10, B=@n22)
  Floor(None=@n23)
  LinearInterpolate(A=Constant(), B=Constant(), Alpha=@n23)
  n24 = ComponentMask(None=@n2.Z Texture) {b=true, g=true, r=true}
  n25 = VectorParameter() {parameterName=Water_Color, defaultValue=<rgba 0.024 0.03 0.0293 0.35>}
  n26 = Constant() {r=0.5}
  n27 = Constant() {r=1}
  n28 = ComponentMask(None=WorldPosition() {worldPositionShaderOffset=WPT_ExcludeAllShaderOffsets}.XYZ) {g=true, r=true}
  n29 = Divide(A=@n28) {constB=200}
  n30 = Constant() {r=0.1}
  n31 = Constant() {r=0.1}
  n32 = Divide(A=@n28) {constB=150}
  n33 = Constant() {r=1}
  n34 = Constant() {r=-1}
  n35 = Add(
   A=MaterialFunctionCall!(
    un frac Time
    Flow Direction=Constant2Vector() {r=1}
    Rotation amount modulator=@n26
    UVS * size=Multiply(A=@n27, B=@n29)
    Time offset
    Grid Noise Tile Rep
    SineAnimSpeed=@n30
    SineAnimOffset
    Flow Speed=@n31
   ) {materialFunction=&/Swamp_TwistFunction}.Rotated UVs
   B=MaterialFunctionCall(
    Flow Direction=Constant2Vector() {g=1}
    Rotation amount modulator=@n26
    UVS * size=Multiply(A=@n32, B=@n33)
    SineAnimSpeed=@n30
    Flow Speed=Multiply(A=@n31, B=@n34)
   ) {materialFunction=&Swamp_TwistFunction}.Rotated UVs
  )
  n36 = Multiply(A=@n35, B=Constant() {r=0.2})
  n37 = Divide(A=ComponentMask(None=WorldPosition() {worldPositionShaderOffset=WPT_ExcludeAllShaderOffsets}.XYZ) {g=true, r=true}) {constB=1024, desc=Normalmap Scale}
  n38 = Multiply(A=Time(), B=Constant() {r=0.02})
  n39 = Add(A=@n37, B=Multiply(A=@n38, B=Constant2Vector() {g=0.8, r=-0.5}))
  n40 = Add(A=@n37, B=Multiply(A=@n38, B=Constant2Vector() {g=-1, r=0.3}))
  n41 = Add(
   A=TextureSample(UVs=Add(A=@n36, B=@n39)) {samplerType=SAMPLERTYPE_Normal, texture=/Game/Packages/DS_Fortnite_TRV_Terrain/Textures/T_TRV_SwampWater_01_N}.RGB
   B=TextureSample(UVs=Add(A=@n36, B=@n40)) {samplerType=SAMPLERTYPE_Normal, texture=&T_TRV_SwampWater_01_N}.RGB
  )
  n42 = ComponentMask(None=@n41) {g=true, r=true}
  n43 = LinearInterpolate(A=Multiply(A=Constant() {r=0.75}, B=@n25.RGB), B=@n25.RGB, Alpha=Clamp(None=Desaturation(None=@n42)) {desc=Mask for water color})
  n44 = LinearInterpolate(
   A=VectorParameter() {parameterName=Duckweed_Color_A, defaultValue=<rgba 0.04 0.0336 0.01 0>}.RGB
   B=VectorParameter() {parameterName=Duckweed_Color_B, defaultValue=<rgba 0.06 0.0431 0.003 0>}.RGB
   Alpha=Clamp(None=Desaturation(None=@n41)) {desc=Mask for Duckweed color}
  )
  n45 = ComponentMask(None=WorldPosition() {worldPositionShaderOffset=WPT_ExcludeAllShaderOffsets}.XYZ) {g=true, r=true}
  n46 = Constant() {r=0.25}
  n47 = Add(
   A=Add(A=Divide(A=@n45) {constB=2048}, B=Multiply(A=@n35, B=@n46))
   B=Multiply(A=Constant() {r=0.005}, B=ComponentMask(None=@n41) {g=true, r=true})
  )
  n48 = Power(
   Base=Multiply(A=TextureSample(UVs=@n47) {texture=&T_TRV_SwampWater_02_D}.A, B=Constant() {r=2.4})
   Exp=Constant() {r=2}
  )
  n49 = OneMinus(None=VertexColor().R)
  n50 = Multiply(A=@n48, B=@n49)
  n51 = Multiply(
   A=Power(Base=Add(A=@n50, B=@n49), Exp=ScalarParameter() {defaultValue=4, parameterName=Pow1})
   B=ScalarParameter() {defaultValue=2, parameterName=Multi1}
  )
  n52 = Power(Base=@n49, Exp=ScalarParameter() {defaultValue=4, parameterName=Pow3})
  n53 = ScalarParameter() {defaultValue=2, parameterName=Multi3}
  n54 = Multiply(A=@n48, B=Clamp(None=Multiply(A=@n52, B=@n53)))
  n55 = Multiply(
   A=Power(Base=Add(A=@n54, B=@n52), Exp=ScalarParameter() {defaultValue=6, parameterName=Pow2})
   B=ScalarParameter() {defaultValue=3, parameterName=Multi2}
  )
  n56 = Clamp(None=Add(A=Clamp(None=@n51) {maxDefault=0.35}, B=Clamp(None=@n55)))
  n57 = LinearInterpolate(
   A=LinearInterpolate(A=@n24, B=ComponentMask(None=@n11.Z Texture) {b=true, g=true, r=true}, Alpha=@n10)
   B=LinearInterpolate(A=@n24, B=LinearInterpolate(A=@n43, B=@n44, Alpha=@n56) {desc=water with duckweed}, Alpha=@n10)
   Alpha=@n22
  )
  n58 = MaterialFunctionCall!(DiffuseFresnelAmount=Constant() {r=0.25}, Emissive, Base Color=@n57) {desc=Fortnite fresnel effect for grounds, materialFunction=/Game/Packages/Fortress_SharedMaterials/Base_Material_Functions/FresnelFunction}
  n59 = LinearInterpolate(A=@n58.d, B=@n57, Alpha=Multiply(A=OneMinus(None=@n56), B=@n22)) {desc=Cut fresnel effect on clear Water}
  MaterialFunctionCall(DiffuseColor=@n59) {materialFunction=&ConvertFromDiffSpec}
  n60 = Constant() {r=0.4}
  n61 = Constant() {r=0.15}
  n62 = Constant() {r=0.2}
  n63 = LinearInterpolate(A=Constant() {r=0.95}, B=@n62, Alpha=NamedRerouteUsage()) {desc=make leaves roughness fade a bit sooner as they look dark far away}
  n64 = Constant3Vector() {constant=<rgba 0 0 1 0>}
  n65 = Multiply(A=@n41, B=Constant3Vector() {desc=Normalmap Amount, constant=<rgba 0.15 0.15 1 0>})
  MaterialFunctionCall!(
   Render To Texture Pass
   WS Flow Map Dir=Constant2Vector() {g=0.5, r=0.5}
   Opacity
   Default Mat Att=MakeMaterialAttributes!(
    BaseColor=@n59
    Metallic=Constant()
    Specular=MaterialFunctionCall("Xb1,PS4"=@n60, "PC-High&&Epic,XSX,PS5"=@n60, "PC-Perf,Android"=@n61, Switch=@n61, "PC-Low(NotPerf)&&Medium"=@n60) {materialFunction=&PlatformSwitch_1D}.Result
    Roughness=LinearInterpolate(A=@n63, B=@n62, Alpha=@n23)
    Anisotropy
    EmissiveColor
    Opacity
    OpacityMask
    Normal=LinearInterpolate(A=@n64, B=@n65, Alpha=@n22) {desc=Cut Normal on ground}
    Tangent
    WorldPositionOffset
    SubsurfaceColor
    ClearCoat
    ClearCoatRoughness
    AmbientOcclusion
    Refraction
    CustomizedUVs_0
    CustomizedUVs_1
    CustomizedUVs_2
    CustomizedUVs_3
    CustomizedUVs_4
    CustomizedUVs_5
    CustomizedUVs_6
    CustomizedUVs_7
    PixelDepthOffset
    ShadingModel
    Displacement
   )
  ) {materialFunction=/Game/WaterAndWind/Materials/Water/WaterRenderToTextureMaterialAtt}
  Multiply(A=LinearInterpolate(B=Constant() {r=0.1}, Alpha=@n23), B=@n22)
  Multiply(A=@n3)
  NamedRerouteDeclaration(None=MaterialFunctionCall!(
    Input=PixelDepth().R
    Input Low=ScalarParameter() {defaultValue=2000, parameterName=Leaves Depth Fade Min Distance}
    Input High=ScalarParameter() {defaultValue=7500, parameterName=Leaves Depth Fade Max Distance}
   ) {materialFunction=/Engine/Functions/Engine_MaterialFunctions03/Math/RemapValueRangeNormalized}.Result) {name=leafFade, nodeColor=<rgba 1 0.5156 0 1>}
  NamedRerouteDeclaration(None=MaterialFunctionCall(
    Input=PixelDepth().R
    Input Low=ScalarParameter() {defaultValue=2000, parameterName=Leaves Depth Fade Min Distance}
    Input High=ScalarParameter() {defaultValue=5500, parameterName=Leaves Depth Fade Max Distance Rough}
   ) {materialFunction=&RemapValueRangeNormalized}.Result) {name=leafFade_Rough, nodeColor=<rgba 1 0.5156 0 1>}
  NamedRerouteUsage()
  OneMinus()
  Subtract()
 ]
}
