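# Masked, lit material for a swamp water border, blending slime-tinted mud into a dimmed waterplant texture along a sharpened, ripple-perturbed V gradient.
# Two world-space normal samples pan in different directions, offset by opposing Swamp_TwistFunction flows, and distort the plant UVs.
# Fresnel shapes mud diffuse and water specular before diffuse/specular conversion.
# Cloud height noise, mud alpha and V gradients form a ragged cutout mask, hard-stepped on one term.
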Material /Game/Packages/DS_Fortnite_TRV_Terrain/Materials/M_TRV_Swamp_Water_Border_02
graph {
 properties {blendMode=BLEND_Masked, pixelDepthOffsetMode=PDOM_Legacy, usedShadingModels=Default Lit}
 nodes [
  Constant()
  n1 = ComponentMask(None=WorldPosition() {worldPositionShaderOffset=WPT_ExcludeAllShaderOffsets}.XYZ) {g=true, r=true}
  n2 = Divide(A=@n1) {constB=2048}
  n3 = Constant() {r=0.5}
  n4 = ComponentMask(None=WorldPosition() {worldPositionShaderOffset=WPT_ExcludeAllShaderOffsets}.XYZ) {g=true, r=true}
  n5 = Multiply(A=Divide(A=@n4) {constB=512}, B=Constant() {r=1})
  n6 = Constant() {r=0.1}
  n7 = Constant() {r=0.01}
  n8 = Multiply(A=Divide(A=@n4) {constB=450}, B=Constant() {r=1})
  n9 = Multiply(A=@n7, B=Constant() {r=-1})
  n10 = Multiply(
   A=Add(
    A=MaterialFunctionCall!(
     un frac Time
     Flow Direction
     Rotation amount modulator=@n3
     UVS * size=@n5
     Time offset
     Grid Noise Tile Rep
     SineAnimSpeed=@n6
     SineAnimOffset
     Flow Speed=@n7
    ) {materialFunction=&/Swamp_TwistFunction}.Rotated UVs
    B=MaterialFunctionCall(Rotation amount modulator=@n3, UVS * size=@n8, Flow Speed=@n9) {materialFunction=&Swamp_TwistFunction}.Rotated UVs
   )
   B=Constant() {r=0.25}
  )
  n11 = Divide(A=ComponentMask(None=WorldPosition() {worldPositionShaderOffset=WPT_ExcludeAllShaderOffsets}.XYZ) {g=true, r=true}) {constB=1024}
  n12 = Multiply(A=Time(), B=Constant() {r=0.02})
  n13 = Add(A=@n11, B=Multiply(A=@n12, B=Constant2Vector() {g=0.8, r=-0.5}))
  n14 = Add(A=@n11, B=Multiply(A=@n12, B=Constant2Vector() {g=-1, r=0.3}))
  n15 = Add(
   A=TextureSample(UVs=Add(A=@n10, B=@n13)) {samplerType=SAMPLERTYPE_Normal, texture=/Game/Packages/DS_Fortnite_TRV_Terrain/Textures/T_TRV_SwampWater_01_N}.RGB
   B=TextureSample(UVs=Add(A=@n10, B=@n14)) {samplerType=SAMPLERTYPE_Normal, texture=&T_TRV_SwampWater_01_N}.RGB
  )
  n16 = Multiply(A=Constant() {r=0.1}, B=@n15)
  n17 = ComponentMask(None=Add(A=Desaturation(None=@n16), B=TextureCoordinate())) {g=true}
  n18 = Constant() {r=9}
  n19 = Clamp(None=Multiply(A=Power(Base=@n17, Exp=@n18), B=Constant() {r=600}))
  n20 = LinearInterpolate(A=Constant(), B=Multiply(A=Constant() {r=0.005}, B=ComponentMask(None=@n15) {g=true, r=true}), Alpha=@n19)
  n21 = Multiply(
   A=TextureSample(UVs=Add(A=@n2, B=@n20)) {texture=/Game/Packages/DS_Fortnite_TRV_Terrain/Textures/T_TRV_Waterplants_03_D}.RGB
   B=Constant() {r=0.4}
  )
  LinearInterpolate(A=Constant(), B=Multiply(A=@n21, B=Constant() {r=0.3}), Alpha=@n19)
  n22 = Constant() {r=0.2}
  n23 = TextureSample(UVs=Divide(A=@n1) {constB=1024}) {texture=/Game/Packages/DS_Fortnite_TRV_Terrain/Textures/T_Transylvania_HF_Mud_1_D}
  n24 = Multiply(A=@n23.RGB, B=VectorParameter() {parameterName=Slime Color, defaultValue=<rgba 0.3865 0.4828 0.3087 0>}.RGB) {constB=0.54}
  n25 = Constant() {r=0.002}
  n26 = Constant() {r=0.03}
  n27 = Multiply(A=Fresnel() {baseReflectFraction=0, exponent=6}, B=OneMinus(None=Constant() {r=0.8}))
  MaterialFunctionCall(
   DiffuseColor=LinearInterpolate(
    A=MaterialFunctionCall!(DiffuseFresnelAmount=@n22, Emissive, Base Color=@n24) {materialFunction=/Game/Packages/Fortress_SharedMaterials/Base_Material_Functions/FresnelFunction}.d
    B=@n21
    Alpha=@n19
   )
   SpecularColor=LinearInterpolate(A=Constant(), B=LinearInterpolate(A=@n25, B=@n26, Alpha=@n27), Alpha=@n19)
  ) {materialFunction=&ConvertFromDiffSpec}
  n28 = OneMinus(None=Multiply(A=@n23.A, B=Constant() {r=1}))
  n29 = TextureSample(UVs=@n2) {samplerType=SAMPLERTYPE_LinearColor, texture=/Game/Packages/Fortress_Sky/Materials/FinalCloudNormalAndHeightMap}
  n30 = ComponentMask(None=Add(A=TextureCoordinate(), B=Constant2Vector() {g=0.35})) {g=true}
  n31 = Add(A=Power(Base=Multiply(A=@n28, B=@n29.R), Exp=Constant() {r=1}), B=@n30)
  n32 = Multiply(A=Power(Base=Multiply(A=@n31, B=@n30), Exp=Constant() {r=1}), B=Constant() {r=2})
  n33 = Divide(A=@n1) {constB=512}
  n34 = Power(
   Base=Multiply(
    A=TextureSample(UVs=@n33) {samplerType=SAMPLERTYPE_LinearColor, texture=&FinalCloudNormalAndHeightMap}.R
    B=Constant() {r=1}
   )
   Exp=Constant() {r=1}
  )
  n35 = OneMinus(None=@n17)
  n36 = Power(Base=Multiply(A=Add(A=@n34, B=@n35), B=@n35), Exp=Constant() {r=1})
  n37 = Constant() {r=2}
  Multiply(A=Clamp(None=Floor(None=@n32)), B=Clamp(None=Multiply(A=@n36, B=@n37)))
  Multiply(A=@n15, B=Constant3Vector() {constant=<rgba 0.15 0.15 1 0>})
 ]
}
